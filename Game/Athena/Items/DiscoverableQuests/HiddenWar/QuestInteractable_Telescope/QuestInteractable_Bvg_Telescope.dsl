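; Quest interactable telescope for Hidden War discoverable quests, offering an "Aim Telescope" interaction that clients block once the objective is completed.
; When client quest visuals spawn, it waits two seconds for ownership, records controllers and visuals, then sets completed players' telescopes to aimed.
; A proximity trigger starts looping hint audio and force feedback for recorded players, stopped on quest disable or completion.
; A server-side list of interacting controllers guards against repeat XP.
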(event EventBeginPlay
  ;[1 1 1] This chunk set telescope to aimed if already interacted with
  (Quests|BindEventtoClientQuestVisualCreated (Variables|Default|GetQuestInteractable) (delegate Custom|OnVisualCreated)))

(event Custom|OnVisualCreated
  (params (QuestVisual :type Quest%20Visual%20Object%20Reference))
  (bind _returnvalue (Actor|GetOwner QuestVisual))
  (bind _returnvalue_2 (Fortnite|GetFortPlayerControllerfromActor _returnvalue))
  ;[1 1 1] This chunk set telescope to aimed if already interacted with
  ;;[1 1 1] wait for ownership assignment
  (Utilities|FlowControl|Delay 2.0)
  (Utilities|Array|AddUnique (Variables|Default|GetPCsOnQuest) _returnvalue_2)
  (bind _returnvalue_3 (Class|QuestInteractableComponent|GetQuestManagerforActor (Variables|Default|GetQuestInteractable) _returnvalue_2))
  (Utilities|Array|AddUnique (Variables|Default|GetQuestVisuals) QuestVisual)
  (Quests|BindEventtoOnQuestsUpdated _returnvalue_3 (delegate Custom|OnQuestsUpdated))
  (for _array_element (Variables|Default|GetQuestVisuals)
    (bind _asquest_visual_bvg_telescope (Utilities|Casting|CastToQuestVisual_Bvg_Telescope _array_element))
    (bind _returnvalue_4 (Class|QuestInteractableComponent|GetQuestManagerforActor (Variables|Default|GetQuestInteractable) (Fortnite|GetFortPlayerControllerfromActor (Actor|GetOwner _array_element))))
    (if (Quest|HasCompletedObjectivewithName _returnvalue_4 (Variables|/Script/FortniteGame.QuestInteractableComponent|GetQuestItemDefinition (Variables|Default|GetQuestInteractable)) (Variables|/Script/FortniteGame.QuestInteractableComponent|GetObjectiveBackendName (Variables|Default|GetQuestInteractable)) :!class /Script/FortniteGame.FortQuestManager)
      (Class|QuestVisualBvgTelescope|SettoAimed _asquest_visual_bvg_telescope))))

(event Custom|OnQuestsUpdated
  ;[1 1 1] This chunk set telescope to aimed if already interacted with
  (for _array_element (Variables|Default|GetQuestVisuals)
    (bind _returnvalue (Class|QuestInteractableComponent|GetQuestManagerforActor (Variables|Default|GetQuestInteractable) (Fortnite|GetFortPlayerControllerfromActor (Actor|GetOwner _array_element))))
    (if (Quest|HasCompletedObjectivewithName _returnvalue (Variables|/Script/FortniteGame.QuestInteractableComponent|GetQuestItemDefinition (Variables|Default|GetQuestInteractable)) (Variables|/Script/FortniteGame.QuestInteractableComponent|GetObjectiveBackendName (Variables|Default|GetQuestInteractable)) :!class /Script/FortniteGame.FortQuestManager)
      _)))

(event ClientQuestEnabled(QuestInteractable))

(event ClientQuestDisabled(QuestInteractable)
  (Audio|Components|Audio|FadeOut (Variables|Default|GetAthena_QuestHint_Loop_Cue) 0.2 0.0 "Linear" :!class /Script/Engine.AudioComponent)
  (ForceFeedback|Stop (Variables|Default|GetForceFeedback)))

(event ClientQuestObjectiveComplete(QuestInteractable)
  (Audio|Components|Audio|FadeOut (Variables|Default|GetAthena_QuestHint_Loop_Cue) 0.2 0.0 "Linear" :!class /Script/Engine.AudioComponent)
  (ForceFeedback|Stop (Variables|Default|GetForceFeedback))
  (Utilities|Array|RemoveItem (Variables|Default|GetPCsOnQuest))
  (for _array_element (Fortnite|GetLocalFortPlayerControllers true)
    (bind _returnvalue (Class|QuestInteractableComponent|GetQuestManagerforActor (Variables|Default|GetQuestInteractable) _array_element))
    (if (Quest|HasCompletedObjectivewithName _returnvalue (Variables|/Script/FortniteGame.QuestInteractableComponent|GetQuestItemDefinition (Variables|Default|GetQuestInteractable)) (Variables|/Script/FortniteGame.QuestInteractableComponent|GetObjectiveBackendName (Variables|Default|GetQuestInteractable)) :!class /Script/FortniteGame.FortQuestManager)
      (Utilities|Array|RemoveItem (Variables|Default|GetPCsOnQuest) _array_element))))

(event Interaction|EventBlueprintOnInteract
  (params (InteractingPawn :type Fort%20Pawn%20Object%20Reference) (InteractionBeingAttempted :type EInteractionBeingAttempted%20Enum :default "FirstInteraction"))
  ;[1 1 1] because this bp was hastily made and is garbagio. this is me just being paranoid that someone will find a way to exploit the xp
  (Utilities|Array|AddUnique (Variables|Default|GetPCsThatCompletedQuest_Server) (Fortnite|GetFortPlayerControllerfromActor InteractingPawn)))

(event OnComponentBeginOverlap(PlayerNearTrigger)
  (params (OverlappedComponent :type Primitive%20Component%20Object%20Reference) (OtherActor :type Actor%20Object%20Reference) (OtherComp :type Primitive%20Component%20Object%20Reference) (OtherBodyIndex :type Integer) (bFromSweep :type Boolean) (SweepResult :type Hit%20Result%20Structure%20%28by%20ref%29))
  ;[1 1 1] Check to see if audio and force feedback should start
  (Utilities|FlowControl|Sequence
    (:then_0
      (bind _asfort_pawn (Utilities|Casting|CastToFortPawn OtherActor))
      (if (> (Utilities|Array|FindItem (Variables|Default|GetPCsOnQuest) (Fortnite|GetFortPlayerControllerfromActor _asfort_pawn)) -1)
        (if (not (Audio|Components|Audio|IsPlaying (Variables|Default|GetAthena_QuestHint_Loop_Cue)))
          (Audio|Components|Audio|Play (Variables|Default|GetAthena_QuestHint_Loop_Cue) 0.0)
          (Components|Activation|Activate (Variables|Default|GetForceFeedback) false))))))

(fn BlueprintCanInteract
  (params (InteractingPawn :type Fort%20Pawn%20Object%20Reference) (InteractionBeingAttempted :type EInteractionBeingAttempted%20Enum :default "FirstInteraction") (InteractionType :type TInteractionType%20Enum :default "IT_NoInteraction"))
  (if (Networking|IsServer)
    (return (== (Utilities|Array|FindItem (Variables|Default|GetPCsThatCompletedQuest_Server)) -1))
    (else
      (bind _returnvalue (Class|QuestInteractableComponent|GetQuestManagerforActor (Variables|Default|GetQuestInteractable) InteractingPawn))
      (return (not (Quest|HasCompletedObjectivewithName _returnvalue (Variables|/Script/FortniteGame.QuestInteractableComponent|GetQuestItemDefinition (Variables|Default|GetQuestInteractable)) (Variables|/Script/FortniteGame.QuestInteractableComponent|GetObjectiveBackendName (Variables|Default|GetQuestInteractable)) :!class /Script/FortniteGame.FortQuestManager))))))

(fn BlueprintGetInteractionString
  (params (InteractingPawn :type Fort%20Pawn%20Object%20Reference) (InteractionBeingAttempted :type EInteractionBeingAttempted%20Enum :default "FirstInteraction"))
  (return "NSLOCTEXT(\"[3D80811A48D19125357392B1FDA01666]\", \"55EB07F94E5A05F8FC2306976318C634\", \"Aim Telescope\")"))

(fn ConstructionScript)
