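; Quest visual for the HiddenWar (Bears vs Gnomes) telescope interactable, animating the scope and its gnome prop.
; Objective completion rotates AimedScope on a timeline, plays giggle feedback, then spawns placeholder FX and destroys the gnome after a short delay.
; SetToAimed runs a second delayed rotation timeline, using a zero axis.
; A helper plays optional sound, FX and force feedback, rumbling only the first local controller despite its Splitscreen name.
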(event EventBeginPlay)

(event Collision|EventActorBeginOverlap
  (params (OtherActor :type Actor%20Object%20Reference)))

(event EventTick
  (params (DeltaSeconds :type Float%20%28single-precision%29)))

(event EventObjectiveCompleted
  (params (PlayerController :type Fort%20Player%20Controller%20Object%20Reference))
  (bind _prop_lawn_ornament_gnome_02 (Variables|Default|GetProp_Lawn_Ornament_Gnome_02))
  (bind (_direction _rotatetime) (Timeline|Timeline_0 PlayFromStart)
    (:Update
      (Transformation|SetRelativeRotation (Variables|Components|GetAimedScope) (Math|Rotator|Lerp(Rotator) "58.000000, 0,0.000000" "0, 0, 0" _rotatetime) false false))
    (:Finished
      ;[1 1 1] Switch to the other node and delete after proper sound plugged in
      (CallFunction|AthenaPlayQuestRewardFeedbackSplitscreen "/Game/Athena/Sounds/Quests/BearsVsGnomes/BvG_Gnomes_Giggle.BvG_Gnomes_Giggle" (Variables|Default|GetFX) (Variables|Default|GetShake) self)
      (Utilities|FlowControl|Delay 0.5)
      (bind _returnvalue (Effects|Components|ParticleSystem|SpawnFXAtLocation "/Game/Athena/Items/DiscoverableQuests/HiddenWar/Placeholder_ForTheGnomes.Placeholder_ForTheGnomes" (Transformation|GetWorldLocation _prop_lawn_ornament_gnome_02)))
      (Components|Activation|Activate _returnvalue true)
      (Components|DestroyComponent _prop_lawn_ornament_gnome_02))))

(event Custom|SetToAimed
  (Utilities|FlowControl|Delay 0.3)
  (bind (_direction _rotatetime) (Timeline|Timeline_1 PlayFromStart)
    (:Update
      (Transformation|SetRelativeRotation (Variables|Components|GetAimedScope) (Math|Vector|RotatorfromAxisandAngle "0, 0, 0" _rotatetime) false false))))

(fn AthenaPlayQuestRewardFeedback_Splitscreen
  (params (NewParam :type Sound%20Base%20Object%20Reference) (NewParam1 :type FXSystem%20Asset%20Object%20Reference) (NewParam2 :type Force%20Feedback%20Effect%20Object%20Reference) (NewParam3 :type Actor%20Object%20Reference))
  (bind _returnvalue (Transformation|GetActorLocation NewParam3))
  (Utilities|IsValid NewParam
    (:Is%20Valid
      (Audio|SpawnSoundatLocation NewParam _returnvalue))
    (:Is%20Not%20Valid
      _))
  (Utilities|IsValid NewParam1
    (:Is%20Valid
      (Effects|Components|ParticleSystem|SpawnFXAtLocation NewParam1 _returnvalue))
    (:Is%20Not%20Valid
      _))
  (Utilities|IsValid NewParam2
    (:Is%20Valid
      (Game|Feedback|ClientPlayForceFeedback (Utilities|Array|Get(acopy) (Fortnite|GetLocalFortPlayerControllers true)) NewParam2 "None" false false false))))

(fn ConstructionScript)
